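# Niagara swing trail for the HangedGrit pickaxe, drawn by one ribbon emitter, Trail_FaceCamera.
# FN anim-trail modules spawn, initialize and update short-lived particles from a system-level skeletal mesh.
# Initial color takes the System.TOD scalar, then RGB scales by a dark, reddening curve because the lerp alpha of 1 bypasses its TOD input.
# Alpha multiplies a quick system-age falloff by a second fade curve.
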NiagaraSystem /BRCosmetics/Weapons/FORT_Melee/Pickaxe_HangedGrit/FX/NS_Pickaxe_HangedGrit_Trail
dependencies {
 usedDataInterfaces [&NiagaraDataInterfaceColorCurve, &NiagaraDataInterfaceCurve, &NiagaraDataInterfaceSkeletalMesh]
 usedDynamicInputs [&FNTimeOfDayScalar]
 usedModules [
  &self:SystemSpawnScript.SystemScriptSource.SystemScriptGraph.NiagaraNodeAssignment_0.SetVariables_#1
  &SystemState
  &self:SystemSpawnScript.SystemScriptSource.SystemScriptGraph.NiagaraNodeAssignment_1.SetVariables_#2
  &self:SystemSpawnScript.SystemScriptSource.SystemScriptGraph.NiagaraNodeAssignment_2.SetVariables_#3
 ]
}
emitters {Trail_FaceCamera {
  emitterUpdateScript [
   &EmitterState()
   &FN_SpawnAnimTrail()
   /SetVariables(Emitter.rand:NiagaraFloat=&RandomRangeFloat(Minimum=1, Maximum=1, Evaluation Type=::NewEnumerator1)) {isSetParameters=true}
  ]
  particleSpawnScript [
   &FN_InitializeAnimTrailBoneIndecies()
   &FN_InitializeAnimTrail(
    Color=/Niagara/DynamicInputs/TypeConversions/MakeLinearColorFromFloat(Value:NiagaraFloat=$System.TOD)
    Lifetime=0.3
    SkeletalMesh=$System.Skeletal Mesh
   )
  ]
  particleUpdateScript [
   &FN_UpdateAnimTrail(SkeletalMesh=$System.Skeletal Mesh)
   &ScaleColor(
    Scale Alpha=&Multiply_Float(
     A=&FloatFromCurve(
      FloatCurve=DataInterface() {propertyValues {Curve {keys [
          {arriveTangent=0.074, interpMode=RCIM_Cubic, leaveTangent=0.074, leaveTangentWeight=0.3333, tangentMode=RCTM_User, time=0, value=1}
          {
           arriveTangent=-11.458
           arriveTangentWeight=0.0492
           interpMode=RCIM_Cubic
           leaveTangent=-11.458
           leaveTangentWeight=0.2841
           tangentMode=RCTM_User
           time=0.1316
           value=0.4947
          }
          {
           arriveTangent=0.0063
           arriveTangentWeight=0.0175
           interpMode=RCIM_Cubic
           leaveTangent=0.0063
           leaveTangentWeight=0.2667
           tangentMode=RCTM_User
           time=0.1839
           value=0
          }
          <at 1 0>
         ]}}}
      CurveIndex=$Engine.System.Age
      Scale Curve=0.2
     )
     B=&FloatFromCurve(
      FloatCurve=DataInterface() {propertyValues {Curve {keys [
          {interpMode=RCIM_Cubic, tangentMode=RCTM_User, time=0, value=1}
          {arriveTangent=-1.4244, interpMode=RCIM_Cubic, leaveTangent=-1.4244, tangentMode=RCTM_User, time=1, value=0}
         ]}}}
      Scale Curve=$Emitter.rand
     )
    )
    Scale RGB=&Lerp_Vector(
     A=&MakeVectorFromLinearColorRGB(Color=&MakeLinearColorFromFloat(Value:NiagaraFloat=$System.TOD))
     B=&MakeVectorFromLinearColorRGB(Color=&ColorFromCurve(ColorCurve=DataInterface() {propertyValues {
         AlphaCurve {keys [<at 0 1 RCIM_Cubic>,<at 0.5108 0 RCIM_Cubic>]}
         BlueCurve {keys [<at 0 0.0783>,<at 0.2842 0.0131>]}
         GreenCurve {keys [<at 0 0.0735>,<at 0.2842 0.0351>]}
         RedCurve {keys [<at 0 0.0885>,<at 0.2842 0.1094>]}
        }}))
     Alpha=1
    )
    ScaleA=true
    ScaleRGB=true
   ) {module=ScaleColor001}
  ]
  rendererClasses [&NiagaraRibbonRendererProperties]
  renderers [{rendererClass=&NiagaraRibbonRendererProperties}]
 }}
summary {
 systemName=NS_Pickaxe_HangedGrit_Trail
 emitters [{emitterName=Trail_FaceCamera, rendererClasses [&NiagaraRibbonRendererProperties]}]
 userVariables [
  User.bIsFrontEnd {type=NiagaraBool}
  User.bIsFrontEndPreview {type=NiagaraBool}
  User.RibbonIDOffset {type=NiagaraInt32}
 ]
}
